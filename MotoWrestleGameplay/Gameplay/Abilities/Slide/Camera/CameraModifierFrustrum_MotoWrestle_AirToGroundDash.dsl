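; Camera modifier for the MotoWrestle air-to-ground dash, part of the Slide ability, that supplies a target camera rotation.
; GetTargetCameraRotation tilts the view target's forward vector around its right axis by the target forward rotation angle and converts it to a rotator.
; Without a valid view target, it falls back to the parent implementation.
; RemoveAfterDelay waits a given time, then removes the modifier from the camera manager.
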(event Custom|RemoveAfterDelay
  (params (Delay :type Float%20%28double-precision%29))
  ; Remove the camera modifier after desired delay.
  (Utilities|FlowControl|Delay Delay)
  (CallFunction|RemoveSelffromCameraManager))

(fn GetTargetCameraRotation
  (bind _returnvalue (CameraModifier|GetViewTarget))
  (Utilities|IsValid _returnvalue
    (:Is%20Valid
      (return (Math|Rotator|MakeRotfromX (Math|Vector|RotateVectorAroundAxis (Transformation|GetActorForwardVector _returnvalue) (Variables|Default|GetTargetForwardRotationAngle) (Transformation|GetActorRightVector _returnvalue)))))
    (:Is%20Not%20Valid
      (bind _returnvalue_2 (|CallParentFunction))
      (return _returnvalue_2))))
