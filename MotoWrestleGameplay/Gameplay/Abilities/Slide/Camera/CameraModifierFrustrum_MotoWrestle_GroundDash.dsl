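; Camera modifier for the MotoWrestle ground dash that supplies a target camera rotation aligned with the view target's facing.
; GetTargetCameraRotation builds that rotation from the view target actor's forward vector, using it as the X axis.
; A RemoveAfterDelay event waits for a given delay, then removes the modifier from the camera manager.
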(event Custom|RemoveAfterDelay
  (params (Delay :type Float%20%28double-precision%29))
  ; Remove the camera modifier after desired delay.
  (Utilities|FlowControl|Delay Delay)
  (CallFunction|RemoveSelffromCameraManager))

(fn GetTargetCameraRotation
  (return (Math|Rotator|MakeRotfromX (Transformation|GetActorForwardVector (CameraModifier|GetViewTarget)))))
